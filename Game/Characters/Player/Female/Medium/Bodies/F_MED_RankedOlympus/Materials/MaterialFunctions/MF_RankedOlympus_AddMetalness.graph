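# Material function for the female medium RankedOlympus player body that adjusts metallic, specular and roughness in masked areas.
# ThinFilmMaskChannel selects a channel of the Tech Art Mask input to scale MetallicIntensity, SpecularOffset and RoughnessOffset.
# Each result is added to the incoming material attribute and saturated.
# The UseCustomAttributes static switch outputs either the modified attributes or the unchanged input.
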MaterialFunction /Game/Characters/Player/Female/Medium/Bodies/F_MED_RankedOlympus/Materials/MaterialFunctions/MF_RankedOlympus_AddMetalness
graph {nodes [
  n1 = FunctionInput() {inputName=Tech Art Mask, sortPriority=10}
  ComponentMask(None=@n1) {b=true}
  n2 = GetMaterialAttributes(None=FunctionInput() {inputName=MaterialAttributes, inputType=FunctionInput_MaterialAttributes}) {attributeGetTypes [#1, #2, #3]}
  n3 = ChannelMaskParameter(None=@n1) {group=05_Tech Art Masks, parameterName=ThinFilmMaskChannel}
  n4 = ScalarParameter() {defaultValue=1, group=RankedOlympus_CustomAttributes, parameterName=MetallicIntensity}
  n5 = Saturate(None=Add(A=@n2.Metallic, B=Multiply(A=@n3.R, B=@n4)))
  n6 = ScalarParameter() {group=RankedOlympus_CustomAttributes, parameterName=SpecularOffset}
  n7 = Saturate(None=Add(A=@n2.Specular, B=Multiply(A=@n3.R, B=@n6)))
  n8 = ScalarParameter() {group=RankedOlympus_CustomAttributes, parameterName=RoughnessOffset}
  n9 = Saturate(None=Add(A=@n2.Roughness, B=Multiply(A=@n3.R, B=@n8)))
  FunctionOutput(None=StaticSwitchParameter(
    True=SetMaterialAttributes!(MaterialAttributes=@n2.MaterialAttributes, Metallic=@n5, Specular=@n7, Roughness=@n9) {attributeSetTypes [#1, #2, #3]}
    False=@n2.MaterialAttributes
   ) {group=RankedOlympus_CustomAttributes, parameterName=UseCustomAttributes})
 ]}
