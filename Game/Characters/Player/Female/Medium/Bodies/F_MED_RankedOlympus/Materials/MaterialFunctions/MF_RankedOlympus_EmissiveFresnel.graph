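# Material function for the RankedOlympus female medium body that sets an emissive rank glow on incoming material attributes, gated by the UseEmissiveRank static switch.
# Fresnel times mask is raised to FresnelPower_01, then scaled by strength, color and noise.
# The noise is a panning Perlin texture with UVs distorted by another panning sample, blended in by NoiseInfluence.
# A sine-driven blend between two Fresnel powers is present but unconnected.
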MaterialFunction /Game/Characters/Player/Female/Medium/Bodies/F_MED_RankedOlympus/Materials/MaterialFunctions/MF_RankedOlympus_EmissiveFresnel
graph {nodes [
  n1 = Multiply(
   A=FunctionInput() {inputName=Fresnel, inputType=FunctionInput_Scalar}
   B=FunctionInput() {inputName=Mask, inputType=FunctionInput_Scalar, sortPriority=10}
  )
  n2 = ScalarParameter() {defaultValue=4, group=EmissiveRankEffect, parameterName=FresnelPower_01}
  n3 = Multiply(A=TextureCoordinate()) {constB=0.5}
  n4 = VectorParameter() {group=EmissiveRankEffect, parameterName=NoiseSpeed}
  n5 = TextureSampleParameter2D(UVs=Panner(Speed=AppendVector(A=@n4.R, B=@n4.R)) {speedX=0.2, speedY=0.12}) {parameterName=ThinFilmUVOffsetTexture, texture=/Game/Global/Textures/Default/Noise/3xPerlin_M}
  n6 = Multiply(A=Add(A=@n5.G, B=@n5.G)) {constB=0.5}
  n7 = TextureSampleParameter2D(UVs=Panner(Coordinate=Add(A=@n3, B=@n6), Speed=AppendVector(A=@n4.B, B=@n4.B))) {parameterName=ThinFilmUVOffsetTexture, texture=&3xPerlin_M}
  n8 = Saturate(None=ScalarParameter() {defaultValue=1, group=EmissiveRankEffect, parameterName=NoiseInfluence})
  n9 = Multiply(
   A=Multiply(A=Power(Base=@n1, Exp=@n2) {constExponent=5}, B=ScalarParameter() {defaultValue=4, group=EmissiveRankEffect, parameterName=EmissiveStrength}) {constB=7}
   B=Multiply(
    A=VectorParameter() {group=EmissiveRankEffect, parameterName=EmissiveColor, defaultValue=<rgba 0.6875 0.1451 0.0903 1>}.RGB
    B=LinearInterpolate(B=@n7.R, Alpha=@n8) {constA=1}
   )
  )
  n10 = Constant()
  FunctionOutput(None=SetMaterialAttributes!(
    MaterialAttributes=FunctionInput() {inputName=MaterialAttributes, inputType=FunctionInput_MaterialAttributes, sortPriority=20}
    EmissiveColor=StaticSwitchParameter(True=@n9, False=@n10) {group=EmissiveRankEffect, parameterName=UseEmissiveRank, sortPriority=0}
   ) {attributeSetTypes [#1]}) {outputName=SetMaterialAttributes}
  n11 = Sine(None=Multiply(A=Time(), B=ScalarParameter() {defaultValue=0.5, group=EmissiveRankEffect, parameterName=FresnelPowerChange_Frequency}))
  LinearInterpolate(
   A=@n2
   B=ScalarParameter() {defaultValue=5, group=EmissiveRankEffect, parameterName=FresnelPower_02}
   Alpha=Add(A=Multiply(A=@n11) {constB=0.5}) {constB=0.5}
  )
 ]}
