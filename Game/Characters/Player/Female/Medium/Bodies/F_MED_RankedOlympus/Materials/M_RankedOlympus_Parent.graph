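# Parent character material for the female medium RankedOlympus body, a ranked cosmetic, using material attributes with a per-expression shading model.
# It wraps a chain of skin-specific effect functions between the shared 2023 character PreFX and PostFX functions.
# The chain adds emissive fresnel, metalness, ranked thin film and a rank icon atlas, in that order.
# The Tech Art Mask feeds each effect, with the ThinFilmMaskChannel parameter choosing the channel that masks the emissive fresnel.
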Material /Game/Characters/Player/Female/Medium/Bodies/F_MED_RankedOlympus/Materials/M_RankedOlympus_Parent
graph {
 properties {
  bUseMaterialAttributes=true
  bUsedWithInstancedStaticMeshes=true
  bUsedWithMeshDeformer=true
  bUsedWithMorphTargets=true
  bUsedWithNanite=true
  bUsedWithSkeletalMesh=true
  pixelDepthOffsetMode=PDOM_Legacy
  shadingModel=MSM_FromMaterialExpression
  usedShadingModels=Default Lit | Subsurface | Clear Coat | Subsurface Profile | Cloth
 }
 nodes [
  n1 = MaterialFunctionCall() {materialFunction=&MF_CharacterParent_2023_PreFX}
  n2 = MaterialFunctionCall!(
   Fresnel=@n1.Fresnel
   Mask=ChannelMaskParameter(None=@n1.Tech Art Mask) {group=05_Tech Art Masks, parameterName=ThinFilmMaskChannel}.R
   MaterialAttributes=@n1.Fresnel
  ) {materialFunction=/Game/Characters/Player/Female/Medium/Bodies/F_MED_RankedOlympus/Materials/MaterialFunctions/MF_RankedOlympus_EmissiveFresnel}
  n3 = MaterialFunctionCall!(MaterialAttributes=@n2.SetMaterialAttributes, Tech Art Mask=@n1.Tech Art Mask) {materialFunction=/Game/Characters/Player/Female/Medium/Bodies/F_MED_RankedOlympus/Materials/MaterialFunctions/MF_RankedOlympus_AddMetalness}
  n4 = MaterialFunctionCall!(
   Material=MaterialFunctionCall!(Material Attributes=@n3.Result, Alpha=@n1.Tech Art Mask) {materialFunction=/Game/Global/Materials/Cosmetics/Ranked/MF_RankedThinFilm}.Result
   Alpha=@n1.Tech Art Mask
  ) {materialFunction=/Game/Characters/Player/Female/Medium/Bodies/F_MED_RankedOlympus/Materials/MaterialFunctions/MF_RankIconAtlas_RankedOlympus}
  MaterialFunctionCall(Material=@n4.Result, M=@n1.M, Pre Skinned Local Position=@n1.M, Distance Blend=@n1.M, Fresnel=@n1.M) {materialFunction=&MF_CharacterParent_2023_PostFX}
 ]
}
